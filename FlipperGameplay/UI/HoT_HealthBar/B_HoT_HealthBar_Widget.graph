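# Heal-over-time health bar overlay widget showing a material-driven frame image.
# On Construct it casts the owning pawn to FortPlayerPawnAthena, binds its health-changed delegate and runs an initial update.
# Each update drives the image's dynamic material GradiantOffset parameter from a negatively scaled health-to-max ratio.
# PreConstruct scales the border's width by EndPercent; StartPercent, Tick and an ease interpolation node are unused.
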WidgetBlueprint /FlipperGameplay/UI/HoT_HealthBar/B_HoT_HealthBar_Widget
graphs [EventGraph {nodes [
   n1 = OnPawnHealthChanged_Event_0() {is=title K2Node_CustomEvent}
   n2 = GetDynamicMaterial(execute=[@n7.then, @n1.then], self=Get Image_HoTFrame() -> (Image_HoTFrame:Image Object Reference))
   n3 = Cast To FortPlayerPawnAthena(execute=@n6.then, Object=GetOwningPlayerPawn!(self:User Widget Object Reference) -> (ReturnValue:Pawn Object Reference))
   n4 = GetMaxHealth!(self:Fort Pawn Object Reference=@n3) -> (ReturnValue:float)
   n5 = GetHealth!(self:Fort Pawn Object Reference=@n3) -> (ReturnValue:float)
   SetScalarParameterValue(
    execute=@n2.then
    self=@n2
    ParameterName=GradiantOffset
    Value=float * float(A=-1.5, B=float / float(A=@n5, B=@n4)) {is=title K2Node_PromotableOperator}
   )
   Event PreConstruct()
   > SetDesiredSizeScale!(
    execute
    self:Border Object Reference=Get Border_0() -> (Border_0:Border Object Reference)
    InScale_X:double=Get EndPercent() -> (EndPercent:double)
    InScale_Y:double=1
   ) -> (then)
   n6 = Event Construct()
   FInterpEaseInOut!(A:double=0, B:double=-70, Alpha:double=0, Exponent:double=20) -> (ReturnValue:double)
   Event Tick()
   n7 = Assign On Pawn Health Changed!(execute:Exec=@n3.then, self:Fort Pawn Object Reference=@n3, Delegate:"Delegate (by ref)"=@n1) -> (then:Exec) {is=title K2Node_AssignDelegate}
  ]}]
variables [
 EndPercent {type=number, default=1}
 StartPercent {type=number, default=0.5}
]
widgets {
 info {inheritedWidgetCount=0, parentClass=&UserWidget, rootWidgetClass=&Overlay, widgetCount=4}
 widgets [
  Overlay_HealOverTime {
   widgetClassPath=&Overlay
   bIsVariable=true
   widget=Overlay_HealOverTime
   properties {renderTransform {angle=0, scale=<1 1>, shear=<0 0>, translation=<4 7>}, renderTransformPivot=<0 0.5>}
  }
  Border_0 {
   widgetClassPath=&Border
   bIsVariable=true
   parent=Overlay_HealOverTime
   widget=Border_0
   properties {
    background {
     bIsDynamicallyLoaded=false
     drawAs=NoDrawType
     imageType=NoImage
     mirroring=NoMirror
     resourceName=None
     resourceObject=null
     tiling=NoTile
     imageSize=<32 32>
     margin=<ltrb 0 0 0 0>
     outlineSettings {
      bUseBrushTransparency=false
      roundingType=HalfHeightRadius
      width=0
      color {colorUseRule=UseColor_Specified, specifiedColor=<rgba 0 0 0 0>}
      cornerRadii=<0 0 0 0>
     }
     tintColor {colorUseRule=UseColor_Specified, specifiedColor=<rgba 1 1 1 1>}
     uVRegion {bIsValid=false, max=<0 0>, min=<0 0>}
    }
    desiredSizeScale=<0.669 1>
    padding=<ltrb 0 0 0 0>
   }
   slot {class=OverlaySlot, properties {horizontalAlignment=HAlign_Fill}}
  }
  SizeBox_0 {
   widgetClassPath=&SizeBox
   parent=Border_0
   widget=SizeBox_0
   properties {bOverride_HeightOverride=true, heightOverride=32, widthOverride=401}
   slot {class=BorderSlot, properties {padding=<ltrb 0 0 0 0>}}
  }
  Image_HoTFrame {
   widgetClassPath=&Image
   bIsVariable=true
   parent=SizeBox_0
   widget=Image_HoTFrame
   properties {
    visibility=HitTestInvisible
    brush {
     bIsDynamicallyLoaded=false
     drawAs=Box
     imageType=NoImage
     mirroring=NoMirror
     resourceName=None
     resourceObject=/FlipperGameplay/UI/Materials/M_UI_HoTHealthBar_Inst
     tiling=NoTile
     imageSize=<10 8>
     margin=<ltrb 0.5 0.5 0.5 0.5>
     outlineSettings {
      bUseBrushTransparency=false
      roundingType=HalfHeightRadius
      width=0
      color {colorUseRule=UseColor_Specified, specifiedColor=<rgba 0 0 0 0>}
      cornerRadii=<0 0 0 0>
     }
     tintColor {colorUseRule=UseColor_Specified, specifiedColor=<rgba 1 1 1 1>}
     uVRegion {bIsValid=false, max=<0 0>, min=<0 0>}
    }
    renderTransformPivot=<0 0.5>
   }
   slot {class=SizeBoxSlot}
  }
 ]
}
